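# Unlit masked Niagara ribbon material for the outer edge trail of the PeelyToonSpring pickaxe, colored by particle color through an emissive adjustment function.
# The edge shape is procedural: an inverted absolute sine of remapped V, smoothstepped by Edge Amount and Edge Softness to isolate both ribbon edges.
# A smoothstepped fade along U is built as a named reroute.
# The mask also includes particle alpha and a near-camera fade.
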Material /BRCosmetics/Weapons/FORT_Melee/Pickaxe_PeelyToonSpring/FX/M_PeelyToonSpring_Outer_Edge_Trail
graph {
 properties {bUsedWithNiagaraRibbons=true, blendMode=BLEND_Masked, pixelDepthOffsetMode=PDOM_Legacy, shadingModel=MSM_Unlit, usedShadingModels=Unlit}
 nodes [
  MaterialFunctionCall(Emissive=ParticleColor().RGB) {materialFunction=&MF_FX_EmissiveAdjustments}
  n1 = ScalarParameter() {defaultValue=0.78, parameterName=Edge Amount}
  n2 = TextureCoordinate()
  n3 = Sine(None=ConstantBiasScale(None=MaterialFunctionCall(Float2=@n2) {materialFunction=&BreakOutFloat2Components}.G))
  n4 = SmoothStep(Min=@n1, Max=Add(A=@n1, B=ScalarParameter() {defaultValue=0.2, parameterName=Edge Softness}), Value=OneMinus(None=Abs(None=@n3)))
  n5 = NamedRerouteUsage()
  Multiply(
   A=Multiply(A=Multiply(A=@n4, B=@n5), B=ParticleColor().A)
   B=MaterialFunctionCall(Distance=ScalarParameter() {defaultValue=64, group=Alpha, parameterName=NearFadeDistance, sortPriority=3}) {materialFunction=&Near_Camera_Fade}.Result
  )
  n6 = ScalarParameter() {defaultValue=0.4, parameterName=OutlineUFade}
  n7 = ScalarParameter() {defaultValue=1, parameterName=OutlineUFade_Softness}
  n8 = MaterialFunctionCall(Float2=TextureCoordinate()) {materialFunction=&BreakOutFloat2Components}
  NamedRerouteDeclaration(None=SmoothStep(Min=@n6, Max=Add(A=@n6, B=@n7), Value=OneMinus(None=@n8.R))) {name=Outline U Fade, nodeColor=<rgba 0 1 0.1094 1>}
 ]
}
